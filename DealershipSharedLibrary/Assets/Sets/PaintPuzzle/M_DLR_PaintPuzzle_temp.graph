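# Masked lit material for a paint puzzle set piece, with animated color, glow, wobble and dissolve effects, usable on skeletal and Niagara meshes.
# Time plus vertex red scrolls a curve atlas lookup, masked by texture, for base color and glow.
# Vertex green offsets noise for eroded opacity with a blue emissive edge and pivot-radial distortion.
# Wobble rotates an offset about an axis; Dynamic Parameters add to scalar Dissolve and Glow.
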Material /DealershipSharedLibrary/Assets/Sets/PaintPuzzle/M_DLR_PaintPuzzle_temp
graph {
 properties {
  bAutomaticallySetUsageInEditor=false
  bUsedWithMeshDeformer=true
  bUsedWithNiagaraMeshParticles=true
  bUsedWithSkeletalMesh=true
  blendMode=BLEND_Masked
  opacityMaskClipValue=0.05
  usedShadingModels=Default Lit
 }
 nodes [
  Add(A=NamedRerouteUsage(), B=NamedRerouteUsage())
  n1 = VectorParameter() {group=Lighting, parameterName=Diffuse Tint, defaultValue=<rgba 1 1 1 0>}
  n2 = Time()
  n3 = ScalarParameter() {defaultValue=0.1, parameterName=Hue Shift Rate}
  n4 = VertexColor()
  n5 = Frac(None=Add(A=Multiply(A=@n2, B=@n3), B=@n4.R))
  n6 = Saturate(None=Multiply(
    A=TextureSample(UVs=@n5) {samplerType=SAMPLERTYPE_LinearColor, texture=&/Curve_PaintPuzzle_Atlas}.RGB
    B=TextureSample() {samplerType=SAMPLERTYPE_Masks, texture=&/T_PaintPuzzle_M}.RGB
   ))
  NamedRerouteDeclaration(None=Multiply(A=Multiply(A=@n6, B=@n1.RGB), B=@n4.B)) {name=Base Color, nodeColor=<rgba 0 0.0391 1 1>}
  n7 = Multiply(A=Time(), B=ScalarParameter() {defaultValue=3, group=Wobble, parameterName=Wobble Speed})
  n8 = Constant3Vector()
  n9 = MaterialFunctionCall(
   Vector=VectorParameter() {group=Wobble, parameterName=Wobble Axis, defaultValue=<rgba 1 0.25 0.5 0>}.RGB
   Default=Constant3Vector() {constant=<rgba 1 0 0 0>}
  ) {materialFunction=&SafeNormalize}
  n10 = Constant3Vector() {constant=<rgba 0 0 1 0>}
  n11 = Multiply(
   A=MaterialFunctionCall(Vector=CrossProduct(A=@n9.Result, B=@n10), Default=Constant3Vector() {constant=<rgba 1 0 0 0>}) {materialFunction=&SafeNormalize}.Result
   B=ScalarParameter() {defaultValue=5, group=Wobble, parameterName=Wobble Amplitude}
  ) {constB=10}
  NamedRerouteDeclaration(None=StaticSwitchParameter(True=RotateAboutAxis(NormalizedRotationAxis=@n9.Result, RotationAngle=@n7, PivotPoint=@n8, Position=@n11), False=Constant3Vector()) {defaultValue=true, group=Wobble, parameterName=Enable Wobble}) {name=Wobble WPO, nodeColor=<rgba 0.3594 1 0 1>}
  n12 = WorldPosition()
  n13 = MaterialFunctionCall() {materialFunction=&ObjectPivotPoint}
  n14 = TextureCoordinate() {uTiling=0.025, vTiling=0.025}
  n15 = Multiply(A=VertexColor().G) {constB=10}
  n16 = Multiply(
   A=MaterialFunctionCall(Vector=Subtract(A=@n12.XYZ, B=@n13.Object Pivot Location)) {materialFunction=&SafeNormalize}.Result
   B=TextureSample(UVs=Add(A=@n14, B=@n15)) {samplerType=SAMPLERTYPE_Normal, texture=/Game/Global/Textures/Default/Noise/64x64Noise}.RGB
  )
  NamedRerouteDeclaration(None=Multiply(A=Multiply(A=@n16) {constB=200}, B=ScalarParameter() {parameterName=Distort})) {name=Dissolve WPO, nodeColor=<rgba 0 0.1328 1 1>}
  n17 = Multiply(
   A=@n6
   B=Add(A=ScalarParameter() {parameterName=Glow}, B=DynamicParameter() {defaultValue=<rgba 0 0 0 0>, paramNames [Dissolve,Glow,na,na]}.Glow)
  )
  n18 = Constant4Vector() {constant=<rgba 0 1 1 0.25>}
  n19 = MaterialFunctionCall(Input=@n4.B, Input Low=@n18.R, Input High=@n18.R, Target Low=@n18.R, Target High=@n18.R) {materialFunction=&RemapValueRange}
  n20 = Constant3Vector() {constant=<rgba 0 0.1237 1 1>}
  n21 = Multiply(A=NamedRerouteUsage()) {constB=2}
  NamedRerouteDeclaration(None=Add(A=Multiply(A=@n17, B=@n19.Result), B=Multiply(A=@n20, B=@n21))) {name=Emission Color, nodeColor=<rgba 0 0.3906 1 1>}
  n22 = TextureCoordinate() {uTiling=0.2, vTiling=0.2}
  n23 = Multiply(A=VertexColor().G) {constB=1.999}
  n24 = MaterialFunctionCall!(
   Erosion=Add(A=ScalarParameter() {parameterName=Dissolve}, B=DynamicParameter() {defaultValue=<rgba 0 0 0 0>, paramNames [Dissolve,Glow,na,na]}.Dissolve)
   Mask=TextureSample(UVs=Add(A=@n22, B=@n23)) {texture=/Game/Global/Textures/Default/Noise/T_Dudebro_Noise_01}.R
   Edge Width
   Width=ScalarParameter() {defaultValue=0.5, parameterName=Dissolve Width}
  ) {materialFunction=/DealershipSharedLibrary/FX/Materials/MaterialFunctions/MF_DLR_Erode}
  NamedRerouteDeclaration(None=@n24.Result) {name=Dissolve Opacity, nodeColor=<rgba 0.0781 0 1 1>}
  NamedRerouteDeclaration(None=OneMinus(None=@n24.Result)) {name=Dissolve Emissionr Mask, nodeColor=<rgba 0 1 0.9531 1>}
  NamedRerouteUsage()
  NamedRerouteUsage()
  NamedRerouteUsage()
  ScalarParameter() {group=Lighting, parameterName=Metallic}
  ScalarParameter() {defaultValue=1, group=Lighting, parameterName=Specular}
  ScalarParameter() {group=Lighting, parameterName=Roughness}
  TextureSample() {samplerType=SAMPLERTYPE_Normal, texture=&/T_PaintPuzzle_N}
 ]
}
